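# Masked, unlit, two-sided FX material for the Donut Plate glider, usable on skeletal and spline meshes.
# World position offset blends sprite facing, spline thickening and noise-driven normal displacement, plus camera offset.
# Normal, polar or spline gradients drive curve-atlas lookups, warped by panning noise, to shape color bands and mask.
# Emissive adds two-sided colors, inner energy, edge detail and core glow; the mask uses fresnel, depth fade and edge falloff.
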Material /Game/Gadgets/Assets/VinderTech_GliderChute/Glider_Donut_Plate/FX/M_DonutPlate
graph {
 properties {
  bUsedWithMeshDeformer=true
  bUsedWithSkeletalMesh=true
  bUsedWithSplineMeshes=true
  blendMode=BLEND_Masked
  pixelDepthOffsetMode=PDOM_Legacy
  shadingModel=MSM_Unlit
  twoSided=true
  usedShadingModels=Unlit
 }
 nodes [
  n1 = TextureCoordinate()
  n2 = Sine(None=Divide(A=ComponentMask(None=@n1) {g=true}))
  n3 = VertexColor()
  n4 = ScalarParameter() {defaultValue=1, group=WPO, parameterName=T_displace_U_tile, sortPriority=5}
  n5 = ScalarParameter() {defaultValue=1, group=WPO, parameterName=T_displace_V_tile, sortPriority=6}
  n6 = Panner(
   Coordinate=Multiply(A=TextureCoordinate(), B=AppendVector(A=@n4, B=@n5))
   Speed=AppendVector(
    A=ScalarParameter() {group=WPO, parameterName=T_displace_U_Speed, sortPriority=7}
    B=ScalarParameter() {defaultValue=1, group=WPO, parameterName=T_displace_V_Speed, sortPriority=8}
   )
  ) {speedY=1}
  n7 = Multiply(
   A=TextureSample(UVs=Add(A=@n3.R, B=@n6)) {samplerType=SAMPLERTYPE_Grayscale, texture=&LowResBlurredNoise}.R
   B=ScalarParameter() {group=WPO, parameterName=Displace_Intensity, sortPriority=3}
  )
  n8 = ScalarParameter() {defaultValue=10, group=WPO, parameterName=SplineWidth, sortPriority=2}
  n9 = LinearInterpolate(
   A=MaterialFunctionCall!(
    Center Location=WorldPosition().XYZ
    0-1 UVs
    XY scale=ScalarParameter() {defaultValue=1, group=WPO, parameterName=Sprite_Scale}
    OverridePivot
    "Normals (Optional)"
   ) {materialFunction=&Sprite}.World position offset
   B=Add(
    A=Multiply(A=@n2, B=@n7)
    B=MaterialFunctionCall(WidthBase=@n8, WidthTip=@n8) {materialFunction=&SplineThicken}.WorldPositionOffset
   )
   Alpha=ScalarParameter() {group=WPO, parameterName="Sprite(0)Spline(1)_WPO", sortPriority=1}
  )
  n10 = ScalarParameter() {defaultValue=1, group=WPO, parameterName=Default_WPO}
  n11 = TextureCoordinate()
  n12 = ComponentMask(None=@n11) {g=true}
  n13 = MaterialFunctionCall(Float2=Sine(None=Divide(A=@n11))) {materialFunction=&BreakOutFloat2Components}
  n14 = Multiply(A=@n13.R, B=@n13.R)
  n15 = VertexNormalWS()
  n16 = Multiply(
   A=LinearInterpolate(A=OneMinus(None=@n12), B=@n14, Alpha=ScalarParameter() {group=WPO, parameterName=Four_Direction, sortPriority=4})
   B=Multiply(A=Normalize(VectorInput=@n15), B=@n7)
  )
  Add(
   A=LinearInterpolate(A=Constant(), B=LinearInterpolate(A=@n9, B=@n16, Alpha=@n10), Alpha=ScalarParameter() {defaultValue=1, parameterName=WPO_Switch})
   B=MaterialFunctionCall(Offset Amount=ScalarParameter() {parameterName=CameraOffset}) {materialFunction=/Engine/Functions/Engine_MaterialFunctions02/WorldPositionOffset/CameraOffset}.Camera Offset
  )
  n17 = DotProduct(A=CameraVectorWS(), B=VertexNormalWS())
  n18 = LinearInterpolate(
   B=ScalarParameter() {defaultValue=1, group=Translucent, parameterName=Fresnel_Enhancement}
   Alpha=Power(Base=Abs(None=@n17), Exp=ScalarParameter() {defaultValue=1, group=Translucent, parameterName=Frenel})
  )
  n19 = MaterialFunctionCall(
   Axis=VectorParameter() {group=Translucent, parameterName=Fresnel_Axis, defaultValue=<rgba 1 0 0 1>}.RGB
   Exponent=ScalarParameter() {defaultValue=1, group=Translucent, parameterName=Frenel_Axis}
  ) {materialFunction=&AxisAlignedFresnel}
  n20 = ComponentMask(None=TextureCoordinate()) {g=true}
  n21 = Multiply(A=@n20, B=ScalarParameter() {defaultValue=5, group=Opacity_Mask, parameterName=Inside_Opacity}) {constB=100}
  n22 = TextureCoordinate()
  n23 = ComponentMask(None=@n22) {r=true}
  n24 = Multiply(A=Abs(None=Subtract(A=@n23) {constB=0.5})) {constB=2}
  n25 = Multiply(A=OneMinus(None=@n24)) {constB=0.5}
  n26 = Multiply(A=MaterialFunctionCall(
    UVs=TextureCoordinate()
    "Rotation Angle (0-1)"=ScalarParameter() {defaultValue=0.25, group=UV_Control, parameterName="UV_Rotate(Polar)", sortPriority=3}
   ) {materialFunction=&CustomRotator}.Rotated Values) {constB=2}
  n27 = ComponentMask(None=MaterialFunctionCall(Vector or UVs=Subtract(A=@n26)) {materialFunction=&VectorToRadialValue}.Radial Coordinates) {r=true}
  n28 = Abs(None=Multiply(A=Subtract(A=@n27) {constB=0.5}) {constB=2})
  n29 = ScalarParameter() {group=UV_Control, parameterName=Polar_Coordinate}
  n30 = OneMinus(None=TextureCoordinate())
  n31 = ComponentMask(None=@n22) {g=true}
  n32 = Multiply(A=ScalarParameter() {defaultValue=0.3, group=UV_Control, parameterName="Slope(Normal)", sortPriority=2}, B=@n24)
  n33 = TextureCoordinate()
  n34 = Constant() {r=0.5}
  n35 = LinearInterpolate(
   A=LinearInterpolate(A=Subtract(A=@n31, B=@n32), B=SphereMask(A=@n33, B=@n34) {attenuationRadius=0.5, hardnessPercent=0}, Alpha=@n29)
   B=OneMinus(None=@n24)
   Alpha=ScalarParameter() {group=WPO, parameterName="Sprite(0)Spline(1)_WPO", sortPriority=1}
  )
  n36 = CurveAtlasRowParameter(CurveTime=@n35) {
   atlas=/Game/Athena/LinearColorCurves/TempAthenaCurveAtlas
   curve=&/Curve_Donut_Plate01
   defaultValue=5
  }
  n37 = AppendVector(
   A=LinearInterpolate(
    A=LinearInterpolate(A=@n25, B=@n28, Alpha=@n29)
    B=ComponentMask(None=@n30) {g=true}
    Alpha=ScalarParameter() {group=WPO, parameterName="Sprite(0)Spline(1)_WPO", sortPriority=1}
   )
   B=@n36.G
  )
  n38 = ScalarParameter() {defaultValue=1, group=Distortion_Ctrl, parameterName=T_Distortion_U_Tile}
  n39 = ScalarParameter() {defaultValue=0.2, group=Distortion_Ctrl, parameterName=T_Distortion_V_Tile}
  n40 = Panner(
   Coordinate=Multiply(A=@n37, B=AppendVector(A=@n38, B=@n39))
   Speed=AppendVector(
    A=ScalarParameter() {defaultValue=-0.5, group=Distortion_Ctrl, parameterName=T_Distortion_U_Speed}
    B=ScalarParameter() {defaultValue=0.3, group=Distortion_Ctrl, parameterName=T_Distortion_V_Speed}
   )
  )
  n41 = AppendVector(A=VertexColor().R, B=Multiply(A=ScalarParameter() {defaultValue=0.05, parameterName=TimeSpeed}, B=Time()))
  n42 = TextureCoordinate()
  n43 = AppendVector(
   A=ScalarParameter() {defaultValue=2, group=Distortion_Ctrl, parameterName=T_Distortion_Sub_U_Tile}
   B=ScalarParameter() {defaultValue=0.5, group=Distortion_Ctrl, parameterName=T_Distortion_Sub_V_Tile}
  )
  n44 = ScalarParameter() {defaultValue=-0.5, group=Distortion_Ctrl, parameterName=T_Distortion_Sub_U_Speed}
  n45 = ScalarParameter() {defaultValue=0.3, group=Distortion_Ctrl, parameterName=T_Distortion_Sub_V_Speed}
  n46 = TextureSample(UVs=Panner(Coordinate=Multiply(A=@n42, B=@n43), Speed=AppendVector(A=@n44, B=@n45)) {speedY=0.3}) {texture=&T_NoisePack_01}
  n47 = ScalarParameter() {defaultValue=0.1, group=Distortion_Ctrl, parameterName=Distortion_Sub_Intensity}
  n48 = TextureSample(UVs=Add(A=Add(A=@n40, B=@n41), B=Multiply(A=@n46.R, B=@n47) {constB=0.1})) {samplerType=SAMPLERTYPE_Normal, texture=/Game/Global/Textures/Default/Tileable_Patterns/T_StormCloudDetail01_N}
  n49 = Multiply(A=@n48.R, B=ScalarParameter() {defaultValue=10, group=Opacity_Mask, parameterName=Mask_Intensity})
  n50 = ScalarParameter() {defaultValue=0.38, group=Distortion_Ctrl, parameterName=Distortion_Intensity}
  n51 = Add(A=@n36.R, B=Multiply(A=@n36.B, B=Multiply(A=@n48.R, B=@n50)) {constA=1}) {constB=0}
  n52 = Multiply(A=@n51, B=ScalarParameter() {defaultValue=2, parameterName=Frequency}) {constB=2}
  n53 = CurveAtlasRowParameter(CurveTime=@n52) {
   atlas=&TempAthenaCurveAtlas
   curve=&/Curve_Donut_Plate02
   defaultValue=6
   parameterName=Curve_DonutPlate_2
  }
  n54 = LinearInterpolate(
   A=Power(Base=@n20, Exp=ScalarParameter() {defaultValue=1, parameterName=Top_Softness})
   B=Constant() {r=1}
   Alpha=ScalarParameter() {group=WPO, parameterName="Sprite(0)Spline(1)_WPO", sortPriority=1}
  )
  n55 = StaticSwitchParameter(
   True=Multiply(A=Power(Base=@n20), B=Add(A=@n21, B=@n49) {constA=2})
   False=LinearInterpolate(A=@n53.A, B=Multiply(A=@n52, B=@n54), Alpha=ScalarParameter() {parameterName=Softness})
  ) {group=Switch, parameterName=PureColor}
  n56 = Multiply(
   A=LinearInterpolate(A=@n18, B=Multiply(A=@n18, B=@n19.Fresnel), Alpha=ScalarParameter() {group=Translucent, parameterName=Fresnel_Type})
   B=DepthFade(Opacity=@n55, FadeDistance=ScalarParameter() {defaultValue=10, group=Translucent, parameterName=Fade_Distance}) {fadeDistanceDefault=10}
  )
  n57 = ShadingPathSwitch(Default=@n56, Deferred=@n56, Mobile=@n55) {desc=Migrated from FeatureLevelSwitch}
  n58 = Power(Base=OneMinus(None=@n14), Exp=ScalarParameter() {defaultValue=1, parameterName=EdgeMask_Power})
  LinearInterpolate(A=@n57, B=Multiply(A=@n57, B=OneMinus(None=@n58)), Alpha=ScalarParameter() {group=Translucent, parameterName=Edge_Mask})
  n59 = Constant() {r=1}
  n60 = CurveAtlasRowParameter(CurveTime=Saturate(None=@n52)) {
   atlas=&TempAthenaCurveAtlas
   curve=&Curve_Donut_Plate02
   defaultValue=6
   parameterName=DarkEdge
  }
  n61 = TextureCoordinate()
  n62 = AppendVector(
   A=ScalarParameter() {defaultValue=1, group=Edge_Detail, parameterName=T_Detail_U_Tile, sortPriority=2}
   B=ScalarParameter() {defaultValue=1, group=Edge_Detail, parameterName=T_Detail_V_Tile, sortPriority=2}
  )
  n63 = ScalarParameter() {group=Edge_Detail, parameterName=T_Detail_U_Speed, sortPriority=2}
  n64 = ScalarParameter() {defaultValue=0.5, group=Edge_Detail, parameterName=T_Detail_V_Speed, sortPriority=2}
  n65 = TextureSampleParameter2D(UVs=Panner(Coordinate=Multiply(A=@n61, B=@n62), Speed=AppendVector(A=@n63, B=@n64)) {speedY=0.1}) {group=Edge_Detail, parameterName=Detail_map, texture=/Game/Global/Textures/Default/Noise/TankMoteNoise}
  n66 = Multiply(A=@n60.B, B=@n65.R)
  n67 = LinearInterpolate(A=ScalarParameter() {group=Edge_Detail, parameterName=Detail_Brightness}, Alpha=Saturate(None=OneMinus(None=@n66)))
  n68 = ScalarParameter() {group=Edge_Detail, parameterName=Edge_Detail_Switch}
  n69 = Multiply(
   A=Power(Base=Saturate(None=@n51), Exp=ScalarParameter() {defaultValue=1, parameterName=Core_Power})
   B=ScalarParameter() {parameterName=Core_Intensity}
  )
  n70 = ScalarParameter() {defaultValue=1, parameterName=Final_Emissive_Intensity}
  n71 = ComponentMask(None=TextureCoordinate()) {g=true}
  n72 = Saturate(None=Subtract(A=OneMinus(None=@n71), B=ScalarParameter() {defaultValue=0.7, group=Spline_Top_White_Ctrl, parameterName=White_Offset}))
  n73 = ScalarParameter() {defaultValue=1, group=Spline_Top_White_Ctrl, parameterName=White_intensity}
  n74 = Add(A=Power(Base=Multiply(A=@n72, B=@n73), Exp=ScalarParameter() {defaultValue=1, group=Spline_Top_White_Ctrl, parameterName=White_Power}))
  n75 = VectorParameter() {parameterName=Dark_Top, defaultValue=<rgba 0.125 0 0.0869 1>}
  n76 = Multiply(A=@n75.RGB, B=ScalarParameter() {defaultValue=0.45, parameterName=Pure_Color_Intensity})
  n77 = CurveAtlasRowParameter(CurveTime=@n35) {
   atlas=&TempAthenaCurveAtlas
   curve=&Curve_Donut_Plate02
   defaultValue=6
   parameterName=Curve_DonutPlate_2
  }
  n78 = CurveAtlasRowParameter(CurveTime=Frac(None=@n52)) {
   atlas=&TempAthenaCurveAtlas
   curve=&Curve_Donut_Plate01
   defaultValue=5
  }
  n79 = VectorParameter() {parameterName=Light_Color_Top, defaultValue=<rgba 1 0.2336 0.986 1>}
  n80 = VectorParameter() {parameterName=Light_Color_Bottom, defaultValue=<rgba 1 1 1 1>}
  n81 = OneMinus(None=SphereMask(A=TextureCoordinate(), B=Constant2Vector() {g=0.3, r=0.5}) {attenuationRadius=0.65, hardnessPercent=0})
  n82 = CurveAtlasRowParameter(CurveTime=@n81) {
   atlas=&TempAthenaCurveAtlas
   curve=&Curve_Donut_Plate02
   defaultValue=6
   parameterName=Curve_DonutPlate_2
  }
  n83 = Add(A=VertexColor().R, B=LinearInterpolate(A=TextureCoordinate(), B=@n37, Alpha=ScalarParameter() {parameterName=InnerEnergy_UVswitch}))
  n84 = Multiply(
   A=AppendVector(
    A=ScalarParameter() {defaultValue=1, group=Inner_Detail, parameterName=T_InnerEnergy_U_Tile}
    B=ScalarParameter() {defaultValue=1, group=Inner_Detail, parameterName=T_InnerEnergy_V_Tile}
   )
   B=@n83
  )
  n85 = AppendVector(
   A=ScalarParameter() {defaultValue=1, group=Inner_Detail, parameterName=T_InnerDetail_U_Tile}
   B=ScalarParameter() {defaultValue=1, group=Inner_Detail, parameterName=T_InnerDetail_V_Tile}
  )
  n86 = ScalarParameter() {group=Inner_Detail, parameterName=T_InnerDetail_U_Speed}
  n87 = ScalarParameter() {defaultValue=1, group=Inner_Detail, parameterName=T_InnerDetail_V_Speed}
  n88 = TextureSample(UVs=Panner(Coordinate=Multiply(A=@n83, B=@n85), Speed=AppendVector(A=@n86, B=@n87)) {speedY=1}) {samplerType=SAMPLERTYPE_LinearGrayscale, texture=/Game/Global/Textures/Default/Noise/T_SwirlNoise_N_height}
  n89 = TextureSampleParameter2D(UVs=Add(A=Panner(Coordinate=@n84) {speedY=0.5}, B=Multiply(A=@n88.R) {constB=0.2})) {
   group=Inner_Detail
   parameterName=T_Inner_Ennergy
   samplerType=SAMPLERTYPE_LinearColor
   texture=/Game/Characters/Player/Male/Medium/Bodies/M_MED_SRDrift/Textures/T_ElectricNoise
  }
  n90 = Clamp(None=Add(A=@n82.G, B=Multiply(A=@n81, B=@n89.R))) {maxDefault=10}
  n91 = Multiply(
   A=Multiply(A=@n77.R) {constB=100}
   B=LinearInterpolate(A=LinearInterpolate(A=@n79.RGB, B=@n80.RGB, Alpha=@n90), B=LinearInterpolate(A=@n75.RGB, B=@n79.RGB, Alpha=@n90), Alpha=@n78.A)
  )
  n92 = LinearInterpolate(A=@n76, B=@n91, Alpha=ScalarParameter() {parameterName=TwoSide_Similarity})
  n93 = Clamp(None=TwoSidedSign())
  n94 = Multiply(
   A=LinearInterpolate(A=@n70, B=Multiply(A=@n74, B=@n70), Alpha=ScalarParameter() {group=WPO, parameterName="Sprite(0)Spline(1)_WPO", sortPriority=1})
   B=StaticSwitchParameter(True=@n76, False=LinearInterpolate(A=@n92, B=@n91, Alpha=@n93)) {group=Switch, parameterName=PureColor}
  )
  MaterialFunctionCall(Emissive=Multiply(A=LinearInterpolate(A=@n59, B=@n67, Alpha=@n68), B=Add(A=@n69, B=@n94))) {materialFunction=&MF_FX_EmissiveAdjustments}
  OneMinus(None=@n65.R)
  Reroute(None=@n55)
  VectorParameter() {parameterName=Dark_Bottom, defaultValue=<rgba 1 0.2338 0.8225 1>}
 ]
}
